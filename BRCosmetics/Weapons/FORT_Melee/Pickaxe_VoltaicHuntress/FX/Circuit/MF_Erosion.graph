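# Material function from the Voltaic Huntress pickaxe FX folder that erodes a scalar gradient by an erosion amount, with adjustable edge softness.
# It subtracts a threshold from the clamped gradient, then divides by softness and saturates for a soft edge.
# The Default Range output thresholds directly on the erosion amount.
# Increased Range stretches the threshold below zero by the softness, so erosion runs from fully solid to fully gone.
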MaterialFunction /BRCosmetics/Weapons/FORT_Melee/Pickaxe_VoltaicHuntress/FX/Circuit/MF_Erosion
graph {
 properties {description=Erodes a gradient and remaps the result according to the softness input, libraryCategoriesText [- TuataraVFX]}
 nodes [
  n1 = TextureCoordinate()
  n2 = Saturate(None=FunctionInput(Preview=ComponentMask(None=@n1) {r=true}) {bUsePreviewValueAsDefault=true, inputName=Gradient, inputType=FunctionInput_Scalar})
  n3 = Saturate(None=FunctionInput(Preview=MaterialFunctionCall() {materialFunction=&DebugTimeSine}.Result) {bUsePreviewValueAsDefault=true, inputName=Erosion Amount, inputType=FunctionInput_Scalar, sortPriority=1})
  n4 = FunctionInput() {bUsePreviewValueAsDefault=true, inputName=Softness, inputType=FunctionInput_Scalar, sortPriority=2, previewValue=<0.1 0 0 1>}
  n5 = Add(A=@n4)
  n6 = Subtract(A=@n2, B=Subtract(A=Multiply(A=@n3, B=@n5), B=@n4))
  FunctionOutput(None=Saturate(None=Divide(A=@n6, B=@n4))) {description=Erosion extrapolating the original range, outputName=Increased Range}
  n7 = Subtract(A=@n2, B=@n3)
  FunctionOutput(None=Saturate(None=Divide(A=@n7, B=@n4))) {description=Erosion retaining the original range, outputName=Default Range}
 ]
}
