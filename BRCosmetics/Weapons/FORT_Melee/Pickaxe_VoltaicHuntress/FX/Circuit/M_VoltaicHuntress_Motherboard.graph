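# Translucent unlit Niagara sprite material for circuit motherboard effects on the Voltaic Huntress pickaxe.
# Emissive blends particle color toward a core color through an exponential radial gradient.
# Opacity comes from masked motherboard texture channels, lerping by isGlow between erosion and circle-SDF radial reveal, times particle alpha.
# A dynamic parameter supplies reveal and erosion amounts via named reroutes; a static switch adds an inverted circle to form a band.
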Material /BRCosmetics/Weapons/FORT_Melee/Pickaxe_VoltaicHuntress/FX/Circuit/M_VoltaicHuntress_Motherboard
graph {
 properties {bUsedWithNiagaraSprites=true, blendMode=BLEND_Translucent, pixelDepthOffsetMode=PDOM_Legacy, shadingModel=MSM_Unlit, usedShadingModels=Unlit}
 nodes [
  n1 = ScalarParameter() {defaultValue=0.4, parameterName=Corecolor}
  n2 = ScalarParameter() {defaultValue=1.5, parameterName=CoreDensity}
  MaterialFunctionCall(Emissive=LinearInterpolate(
    A=ParticleColor().RGB
    B=VectorParameter() {parameterName=Corecolor, defaultValue=<rgba 1.601 4 0 1>}.RGB
    Alpha=MaterialFunctionCall(Radius=@n1, Density=@n2) {materialFunction=&RadialGradientExponential}.RadialGradientExponential
   )) {materialFunction=&MF_FX_EmissiveAdjustments}
  n3 = Subtract(A=NamedRerouteUsage(), B=ScalarParameter() {defaultValue=1, parameterName=RevealSize}) {constB=0.2}
  n4 = NamedRerouteUsage()
  n5 = ScalarParameter() {defaultValue=0.05, parameterName=GlowSize}
  n6 = TextureSampleParameter2D() {
   parameterName=Texture
   samplerType=SAMPLERTYPE_Masks
   texture=&/T_Motherboard_Fortnite
  }
  n7 = Multiply(
   A=MaterialFunctionCall(Size=Subtract(A=@n4, B=@n5) {constB=0.2}) {materialFunction=&MF_UI_Circle}.SDF
   B=StaticComponentMaskParameter(None=@n6.RGB) {defaultG=true, parameterName=RadialErosion}
  )
  n8 = Multiply(A=OneMinus(None=MaterialFunctionCall(Size=@n3) {materialFunction=&MF_UI_Circle}.SDF), B=@n7)
  n9 = NamedRerouteUsage()
  n10 = ScalarParameter() {defaultValue=0.01, parameterName="Softness (S)"}
  n11 = StaticComponentMaskParameter(None=@n6.RGB) {defaultR=true, parameterName=Erosion}
  Multiply(
   A=LinearInterpolate(
    A=MaterialFunctionCall!(Gradient=@n11, Erosion Amount=@n9, Softness=@n10) {materialFunction=&/MF_Erosion}.Default Range
    B=StaticSwitchParameter(True=@n8, False=@n7) {parameterName=RevealErosion}
    Alpha=ScalarParameter() {parameterName=isGlow}
   )
   B=ParticleColor().A
  )
  n12 = DynamicParameter() {defaultValue=<rgba 1 0 1 1>, paramNames [RevealErosion, Erosion, Param3, Param4]}
  NamedRerouteDeclaration(None=@n12.RevealErosion) {name=RevealErosion, nodeColor=<rgba 1 0.4922 0 1>}
  NamedRerouteDeclaration(None=@n12.Erosion) {name=Erosion, nodeColor=<rgba 0.1719 0 1 1>}
 ]
}
