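# Material function for the Voltaic Huntress pickaxe's holographic circuit FX that outputs UVs with a blocky, animated pixelation glitch.
# It rotates input UVs and floors them into a grid, hashed with time for per-cell random values.
# Those values choose between two pixelation resolutions and mask which cells take the quantized UVs.
# A second time-based hash, rounded to on/off, intermittently restores the original UVs everywhere.
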MaterialFunction /BRCosmetics/Weapons/FORT_Melee/Pickaxe_VoltaicHuntress/FX/Circuit/Functions/MF_Holog_Glitch_Grid
graph {nodes [
  n1 = FunctionInput(Preview=TextureCoordinate()) {bUsePreviewValueAsDefault=true, inputName=UV, inputType=FunctionInput_Vector2}
  n2 = MaterialFunctionCall(
   UVs=@n1
   Rotation Center=Constant2Vector() {g=0.5, r=0.5}
   "Rotation Angle (0-1)"=ScalarParameter() {group=HolographicTexture_Glitch, parameterName=GlitchUVRotation}
  ) {materialFunction=&CustomRotator}
  n3 = Time()
  n4 = Multiply(A=@n2.Rotated Values, B=ScalarParameter() {defaultValue=2.5, group=HolographicTexture_Glitch, parameterName=GlitchMaskResolution})
  n5 = MaterialFunctionCall!(Seed=Multiply(A=@n3), Float, Vec3, Vec2=Floor(None=Add(A=@n4))) {materialFunction=&/MF_Hash}
  n6 = LinearInterpolate(
   A=ScalarParameter() {defaultValue=12, group=HolographicTexture_Glitch, parameterName=PixellateGlitch_MaxResolution}
   B=ScalarParameter() {defaultValue=4, group=HolographicTexture_Glitch, parameterName=PixellateGlitch_MinResolution}
   Alpha=Round(None=Clamp(None=@n5.Float))
  )
  n7 = Divide(A=Floor(None=Multiply(A=@n2.Rotated Values, B=@n6)), B=@n6)
  n8 = Round(None=ComponentMask(None=ConstantBiasScale(None=@n5.Vec2)) {r=true})
  n9 = Add(A=@n3) {constB=0.5}
  n10 = Constant() {r=2}
  n11 = ConstantBiasScale(None=MaterialFunctionCall(Float=Multiply(A=@n9, B=@n10)) {materialFunction=&MF_Hash}.Float)
  FunctionOutput(None=LinearInterpolate(A=LinearInterpolate(A=@n1, B=@n7, Alpha=@n8), B=@n1, Alpha=Round(None=@n11)))
 ]}
